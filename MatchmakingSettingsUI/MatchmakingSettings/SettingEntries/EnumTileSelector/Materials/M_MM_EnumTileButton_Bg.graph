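# Translucent UI material for the background of enum tile selector buttons in the matchmaking settings screen.
# A Texture parameter is sampled with UVs from the ScaleY output of MF_UI_UVScaling.
# The IsDisabled scalar blends the sampled color toward a darkened copy of itself.
# MF_UI_CoreUI_RoundedMask supplies a rounded-corner mask with a fixed radius in pixels.
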Material /MatchmakingSettingsUI/MatchmakingSettings/SettingEntries/EnumTileSelector/Materials/M_MM_EnumTileButton_Bg
graph {
 properties {bAutomaticallySetUsageInEditor=false, blendMode=BLEND_Translucent, materialDomain=MD_UI, usedShadingModels=Default Lit}
 nodes [
  n1 = TextureSample(
   UVs=MaterialFunctionCall() {materialFunction=&MF_UI_UVScaling}.ScaleY
   Tex=TextureObjectParameter() {parameterName=Texture}
  )
  LinearInterpolate(A=@n1.RGB, B=Multiply(A=@n1.RGB) {constB=0.6}, Alpha=ScalarParameter() {parameterName=IsDisabled, sliderMax=1})
  MaterialFunctionCall(Corner Radius=Constant() {r=10}, Radius in Pixels=Constant() {r=1}) {materialFunction=&MF_UI_CoreUI_RoundedMask}
 ]
}
