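# Tile button for an option in the matchmaking settings enum tile selector, built on UIKitModularButton with a bottom label strip over a material background.
# Selection plays Anim_IsSelected forward or in reverse, blending label color and the strip material's IsSelected parameter, while hover drives the UIKit outline's focus visuals.
# SetDisabledState dims the label and sets IsDisabled on both dynamic materials cached at init.
# PreConstruct sizes the spacer by compact UI mode.
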WidgetBlueprint /MatchmakingSettingsUI/MatchmakingSettings/SettingEntries/EnumTileSelector/WBP_MM_EnumTileButton
animations [Anim_IsSelected {
  binding_count=2
  length_seconds=0.2
  object=&self:Anim_IsSelected
  sequence {
   bindings [
    Txt_Label {
     id=#1
     possessed_class=CommonTextBlock
     tracks [{
       type=MovieSceneColorTrack
       display_name=Color and Opacity
       sections [{
         type=MovieSceneColorSection
         end_frame=12
         start_frame=0
         channels [
          Color.R {type=MovieSceneScriptingFloatChannel, keys [[0,0.5776],[12,0.0048]]}
          Color.G {type=MovieSceneScriptingFloatChannel, keys [[0,0.5776],[12,0.004]]}
          Color.B {type=MovieSceneScriptingFloatChannel, keys [[0,0.5776],[12,0.0116]]}
          Color.A {type=MovieSceneScriptingFloatChannel, const [1,2]}
         ]
        }]
      }]
    }
    Img_TextBg {
     id=#2
     possessed_class=Image
     tracks [{
       type=MovieSceneWidgetMaterialTrack
       sections [{
         type=MovieSceneParameterSection
         end_frame=12
         start_frame=0
         channels [IsSelected {type=MovieSceneScriptingFloatChannel, keys [[0,0],[12,1]]}]
        }]
      }]
    }
   ]
   display_rate {numerator=60}
  }
 }]
graphs [
 EventGraph {nodes [
   Event OnInitialized()
   > SetInitVisual()
   Event On Selected()
   n1 = Get Anim_IsSelected() -> (Anim_IsSelected:Widget Animation Object Reference)
   > PlayAnimationForward(InAnimation=@n1)
   Event PreConstruct()
   > SetMobileStyle() {is=title K2Node_CallFunction}
   Event On Hovered()
   n2 = Get WBP_UIKit_Block_Outline() -> (WBP_UIKit_Block_Outline:WBP UIKit Block Outline Object Reference)
   > Focused_Visualization!(execute, self:WBP UIKit Block Outline Object Reference=@n2, NeedsToAnimate:bool=true) -> (then)
   Event On Unhovered()
   > Unfocused_Visualization!(execute, self:WBP UIKit Block Outline Object Reference=@n2, NeedsToAnimate:bool=true) -> (then)
   Event On Deselected()
   > PlayAnimationReverse(InAnimation=@n1)
  ]}
 SetDisabledState {nodes [
   n1 = SetDisabledState!() -> (then, IsDisabled:bool=~) {is=title K2Node_FunctionEntry}
   > Set RenderOpacity!(
    execute
    RenderOpacity:float=Select(Option 0:float=1, Option 1:float=0.6, Index=@n1) -> (ReturnValue:float)
    self:Widget Object Reference=Get Txt_Label() -> (Txt_Label:Common Text Object Reference)
   ) -> (then, Output_Get:float)
   > SetScalarParameterValue(
    self=[
     Get DynamicTextBg() -> (DynamicTextBg:Material Instance Dynamic Object Reference)
     Get DynamicBg() -> (DynamicBg:Material Instance Dynamic Object Reference)
    ]
    ParameterName=IsDisabled
    Value=Select(Option 0=0, Option 1=1, Index=@n1)
   )
  ]}
 SetInitVisual {nodes [
   SetInitVisual() {is=title K2Node_FunctionEntry}
   > n1 = GetDynamicMaterial(self=Get Img_TextBg() -> (Img_TextBg:Image Object Reference))
   > Set DynamicTextBg(DynamicTextBg:Material Instance Dynamic Object Reference=@n1) -> (Output_Get:Material Instance Dynamic Object Reference)
   > n2 = GetDynamicMaterial(self:Border Object Reference=Get B_Setting() -> (B_Setting:Common Border Object Reference))
   > Set DynamicBg(DynamicBg:Material Instance Dynamic Object Reference=@n2) -> (Output_Get:Material Instance Dynamic Object Reference)
  ]}
 SetMobileStyle {nodes [
   SetMobileStyle() {is=title K2Node_FunctionEntry}
   > Set Size!(
    execute
    Size:Vector 2D Structure=Select(Option 0:Vector 2D Structure=(X=256,Y=128), Option 1:Vector 2D Structure=(X=352,Y=184), Index=IsUsingCompactUI()) -> (ReturnValue:Vector 2D Structure)
    self:Spacer Object Reference=Get Min_Size() -> (Min_Size:Spacer Object Reference)
   ) -> (then, Output_Get:Vector 2D Structure)
  ]}
]
variables [
 DynamicBg {type=MaterialInstanceDynamic}
 DynamicTextBg {type=MaterialInstanceDynamic}
]
widgets {
 info {inheritedWidgetCount=1, parentClass=/Script/FortniteUI.UIKitModularButton, rootWidgetClass=&CommonBorder, widgetCount=8}
 widgets [
  B_Setting {
   widgetClassPath=&CommonBorder
   bIsVariable=true
   widget=B_Setting
   properties {
    background {
     bIsDynamicallyLoaded=false
     drawAs=Image
     imageType=FullColor
     mirroring=NoMirror
     resourceName=None
     resourceObject=/MatchmakingSettingsUI/MatchmakingSettings/SettingEntries/EnumTileSelector/MaterialInstances/MI_MM_EnumTileButton_Bg
     tiling=NoTile
     imageSize=<32 32>
     margin=<ltrb 0 0 0 0>
     outlineSettings {
      bUseBrushTransparency=false
      roundingType=HalfHeightRadius
      width=0
      color {colorUseRule=UseColor_Specified, specifiedColor=<rgba 0 0 0 0>}
      cornerRadii=<0 0 0 0>
     }
     tintColor {colorUseRule=UseColor_Specified, specifiedColor=<rgba 1 1 1 1>}
     uVRegion {bIsValid=false, max=<0 0>, min=<0 0>}
    }
    padding=<ltrb 0 0 0 0>
   }
  }
  O_SettingContainer {
   widgetClassPath=&Overlay
   parent=B_Setting
   widget=O_SettingContainer
   properties {bOverride_Cursor=true}
   slot {class=BorderSlot, properties {padding=<ltrb 0 0 0 0>}}
  }
  Min_Size {
   widgetClassPath=&Spacer
   bIsVariable=true
   parent=O_SettingContainer
   widget=Min_Size
   properties {size=<256 128>}
   slot {class=OverlaySlot, properties {horizontalAlignment=HAlign_Fill, verticalAlignment=VAlign_Fill}}
  }
  O_LabelContainer {
   widgetClassPath=&Overlay
   parent=O_SettingContainer
   widget=O_LabelContainer
   slot {class=OverlaySlot, properties {horizontalAlignment=HAlign_Fill, verticalAlignment=VAlign_Bottom, padding=<ltrb 2 2 2 2>}}
  }
  Img_TextBg {
   widgetClassPath=&Image
   bIsVariable=true
   parent=O_LabelContainer
   widget=Img_TextBg
   properties {
    visibility=SelfHitTestInvisible
    brush {
     bIsDynamicallyLoaded=false
     drawAs=Image
     imageType=FullColor
     mirroring=NoMirror
     resourceName=None
     resourceObject=/MatchmakingSettingsUI/MatchmakingSettings/SettingEntries/EnumTileSelector/MaterialInstances/MI_MM_EnumTileButton_TextBg
     tiling=NoTile
     imageSize=<32 32>
     margin=<ltrb 0 0 0 0>
     outlineSettings {
      bUseBrushTransparency=false
      roundingType=HalfHeightRadius
      width=0
      color {colorUseRule=UseColor_Specified, specifiedColor=<rgba 0 0 0 0>}
      cornerRadii=<0 0 0 0>
     }
     tintColor {colorUseRule=UseColor_Specified, specifiedColor=<rgba 1 1 1 1>}
     uVRegion {bIsValid=false, max=<0 0>, min=<0 0>}
    }
   }
   slot {class=OverlaySlot, properties {horizontalAlignment=HAlign_Fill, verticalAlignment=VAlign_Fill}}
  }
  Txt_Label {
   widgetClassPath=&CommonTextBlock
   bIsVariable=true
   parent=O_LabelContainer
   widget=Txt_Label
   properties {
    autoWrapText=true
    text=$Label
    colorAndOpacity {colorUseRule=UseColor_Specified, specifiedColor=<rgba 0.5776 0.5776 0.5776 1>}
    font {
     bForceMonospaced=false
     bMaterialIsStencil=false
     fontMaterial=null
     fontObject=&NotoSans
     letterSpacing=0
     monospacedWidth=1
     size=24
     skewAmount=0
     typefaceFontName=NotoSans-Regular
     outlineSettings {
      bApplyOutlineToDropShadows=false
      bMiteredCorners=false
      bSeparateFillAlpha=false
      outlineMaterial=null
      outlineSize=0
      outlineColor=<rgba 0 0 0 1>
     }
    }
    shadowOffset=<0 0>
   }
   slot {class=OverlaySlot, properties {horizontalAlignment=HAlign_Fill, verticalAlignment=VAlign_Fill, padding=<ltrb 24 4 24 4>}}
  }
  WBP_UIKit_Block_Outline {
   widgetClassPath=/FortUILibrary/UIKit/Buttons/Blocks/WBP_UIKit_Block_Outline
   bIsVariable=true
   parent=O_SettingContainer
   widget=WBP_UIKit_Block_Outline
   properties {bottomLeftCornerRadius=20, bottomRightCornerRadius=20, radiusInPixels=true, topLeftCornerRadius=20, topRightCornerRadius=20}
   slot {class=OverlaySlot, properties {horizontalAlignment=HAlign_Fill, verticalAlignment=VAlign_Fill}}
  }
  InputActionWidget {widgetClassPath=&CommonActionWidget, bInherited=true, widget=null}
 ]
}
