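# WBP_MM_EnumTileSelector is a CommonUserWidget that lays out a horizontal row of tile buttons, named as an enum selector entry for the matchmaking settings UI.
# A material-driven background image fills the root overlay behind a padded, clipped overlay holding the CommonListView ListView_TileButtons, which spawns WBP_MM_EnumTileButton entries.
# The list returns focus to the selection and does not select on navigation.
# The event graph is empty, so the widget itself contains no population or selection logic.
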WidgetBlueprint /MatchmakingSettingsUI/MatchmakingSettings/SettingEntries/EnumTileSelector/WBP_MM_EnumTileSelector
graphs [EventGraph {links [], nodes []}]
variables []
widgets {
 info {inheritedWidgetCount=0, parentClass=&CommonUserWidget, rootWidgetClass=&Overlay, widgetCount=4}
 widgets [
  Overlay_27 {widgetClassPath=&Overlay, widget=Overlay_27}
  Background {
   widgetClassPath=&Image
   parent=Overlay_27
   widget=Background
   properties {
    visibility=SelfHitTestInvisible
    brush {
     bIsDynamicallyLoaded=false
     drawAs=Image
     imageType=FullColor
     mirroring=NoMirror
     resourceName=None
     resourceObject=/MatchmakingSettingsUI/MatchmakingSettings/MaterialInstances/MI_UI_MatchmakingSettings_Entry_Background
     tiling=NoTile
     imageSize=<32 32>
     margin=<ltrb 0 0 0 0>
     outlineSettings {
      bUseBrushTransparency=false
      roundingType=HalfHeightRadius
      width=0
      color {colorUseRule=UseColor_Specified, specifiedColor=<rgba 0 0 0 0>}
      cornerRadii=<0 0 0 0>
     }
     tintColor {colorUseRule=UseColor_Specified, specifiedColor=<rgba 1 1 1 1>}
     uVRegion {bIsValid=false, max=<0 0>, min=<0 0>}
    }
   }
   slot {class=OverlaySlot, properties {horizontalAlignment=HAlign_Fill, verticalAlignment=VAlign_Fill}}
  }
  Overlay_0 {
   widgetClassPath=&Overlay
   parent=Overlay_27
   widget=Overlay_0
   properties {bOverride_Cursor=true, clipping=ClipToBounds}
   slot {class=OverlaySlot, properties {horizontalAlignment=HAlign_Fill, verticalAlignment=VAlign_Fill, padding=<ltrb 4 4 4 4>}}
  }
  ListView_TileButtons {
   widgetClassPath=&CommonListView
   bIsVariable=true
   parent=Overlay_0
   widget=ListView_TileButtons
   properties {
    bReturnFocusToSelection=true
    bSelectItemOnNavigation=false
    clipping=Inherit
    consumeMouseWheel=Never
    entryWidgetClass=&/WBP_MM_EnumTileButton.WBP_MM_EnumTileButton_C
    horizontalEntrySpacing=12
    orientation=Orient_Horizontal
   }
   slot {class=OverlaySlot, properties {horizontalAlignment=HAlign_Fill, verticalAlignment=VAlign_Fill, padding=<ltrb 4 4 4 4>}}
  }
 ]
}
